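# Collapsible category side navigation for the quest modal, wrapping the UIKit SideNavigation widget.
# A conversion function orders tracker, available and unavailable category view models; initialization builds filtered tabs from them, and dispatchers report clicks, selection, collapse and visibility need.
# A reversible animation of SizeTransition eases root width toward a minimum set by scrollbar visibility and compact UI.
# Gamepad focus expands it; an oversized invisible touch-only catcher supports tap-outside collapsing.
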WidgetBlueprint /QuestUI/QuestScreen/Modal/WBP_QuestModal_SideNavigation
animations [Anim_Collapse {
  binding_count=1
  length_seconds=1
  object=&self:Anim_Collapse
  sequence {bindings [WBP_QuestModal_SideNavigation_C_28 {
     id=#1
     possessed_class=WBP_QuestModal_SideNavigation_C
     tracks [{
       type=MovieSceneDoubleTrack
       display_name=Size Transition
       sections [{type=MovieSceneDoubleSection, end_frame=20, start_frame=0, channels [None {type=MovieSceneScriptingDoubleChannel, keys [[0,0],[20,1]]}]}]
      }]
    }]}
 }]
graphs [
 CF_GetSideNavVMs {nodes [
   CF_GetSideNavVMs!() -> (then, QuestUIProductVM_In:Quest UI Product View Model Object Reference, QuestUIPlayerVM_In:Quest UI Player View Model Object Reference) {is=title K2Node_FunctionEntry}
   > n1 = Sequence()
   > n2 = Get!(execute:Exec) -> (then:Exec, else:Exec, QuestUIPlayerVM_In:Quest UI Player View Model Object Reference)
   > n3 = Add(
    TargetArray:Array of Quest UI Category Interface View Model Object References=Get SideNavOut() -> (SideNavOut:Array of Quest UI Category Interface View Model Object References)
    NewItem:"Quest UI Category Interface View Model Object Reference (by ref)"=Get TrackerVM!(self:Quest UI Player View Model Object Reference=@n2) -> (TrackerVM:Quest UI Tracker View Model Object Reference)
   ) {is=title K2Node_CallArrayFunction}
   n4 = Get!(execute=[@n3.then, @n2.else]) -> (then, else, QuestUIProductVM_In:Quest UI Product View Model Object Reference)
   > n5 = For Each Loop(Array:Array of Quest UI Category Row View Model Object References=Get OrderedAvailableRowVMs!(self:Quest UI Product View Model Object Reference=@n4) -> (OrderedAvailableRowVMs:Array of Quest UI Category Row View Model Object References)) -> (Array Element:Quest UI Category Row View Model Object Reference)
   > Append Array!(
    execute
    TargetArray:Array of Quest UI Category Interface View Model Object References=Get SideNavOut()
    SourceArray:Array of Quest UI Category Interface View Model Object References=Get OrderedCategoryInterfaceVMs!(self:Quest UI Category Row View Model Object Reference=@n5) -> (OrderedCategoryInterfaceVMs:Array of Quest UI Category Interface View Model Object References)
   ) -> (then)
   n6 = For Each Loop(
    Exec=@n5.Completed
    Array=Get OrderedUnavailableRowVMs!(self:Quest UI Product View Model Object Reference=@n4) -> (OrderedUnavailableRowVMs:Array of Quest UI Category Row View Model Object References)
   )
   > Append Array(TargetArray=Get SideNavOut(), SourceArray=Get OrderedCategoryInterfaceVMs(self=@n6))
   Return Node!(execute=@n1.then_1, SideNavVMs:Array of Quest UI Category Interface View Model Object References=Get SideNavOut())
  ]}
 EventGraph {nodes [
   n1 = "On Side Navigation Collapsed State Changed (SideNavigation)"!() -> (OutputDelegate, then, IsCollapsed:bool) {is=title K2Node_ComponentBoundEvent}
   > HandleCollapseRequest!(execute, self, ShouldCollapse:bool=@n1.IsCollapsed, IsImmediate:bool=false) -> (then)
   > Call On Collapsed State Changed!(execute:Exec, self:WBP Quest Modal Side Navigation Object Reference, IsCollapsed:bool=@n1.IsCollapsed) -> (then:Exec) {is=title K2Node_CallDelegate}
   > PlayTransition!(execute, self, IsCollapsed:bool=@n1.IsCollapsed, SkipAnimation:bool=false) -> (then)
   > SetVisibility(
    self=Get TouchCatcher_VisualAttachment() -> (TouchCatcher_VisualAttachment:Common Visual Attachment Object Reference)
    InVisibility=Select(Option 0=SelfHitTestInvisible, Option 1=Collapsed, Index=@n1.IsCollapsed)
   )
   n2 = "On Tab Selected (SideNavigation)"!() -> (OutputDelegate, then, TabId:Name=None) {is=title K2Node_ComponentBoundEvent}
   n3 = Get SideNavigation() -> (SideNavigation:WBP UIKit Side Navigation Object Reference)
   > ScrollToWidget!(
    execute
    self:WBP UIKit Side Navigation Base Object Reference=@n3
    Widget:Widget Object Reference=GetTabButtonBaseByID!(self:Common Tab List Widget Base Object Reference=@n3, TabNameID:Name=@n2.TabId) -> (ReturnValue:Common Button Base Object Reference)
    IsDelayed:bool=NOT Boolean(A=Get SkipExpandAnimation!(self:WBP UIKit Side Navigation Base Object Reference=@n3) -> (SkipExpandAnimation:bool))
   ) -> (then)
   n4 = "On Tab Clicked (SideNavigation)"!() -> (OutputDelegate, then, TabId:Name=None) {is=title K2Node_ComponentBoundEvent}
   > n5 = Cast To QuestUICategoryInterfaceVM!(
    execute
    Object=GetTabViewModel!(self:UIKit Tab Group Object Reference=Get SideNavigation(), TabID:Name=@n4.TabId) -> (ReturnValue:MVVM Base Viewmodel Object Reference)
   ) -> (then, CastFailed, AsQuest UI Category Interface View Model:Quest UI Category Interface View Model Object Reference) {is=title K2Node_DynamicCast}
   > Call On Category Clicked!(
    execute:Exec
    self:WBP Quest Modal Side Navigation Object Reference
    ClickedCategory:Quest UI Category Interface View Model Object Reference=@n5
   ) -> (then:Exec) {is=title K2Node_CallDelegate}
   n6 = "On Side Navigation Selection Changed (SideNavigation)"!() -> (OutputDelegate, then, TabID:Name=None, InnerTabID:Name=None, TabViewModel:MVVM Base Viewmodel Object Reference) {is=title K2Node_ComponentBoundEvent}
   > n7 = Cast To QuestUICategoryInterfaceVM(Object=@n6.TabViewModel)
   > Call On Category Selected!(execute:Exec, self:WBP Quest Modal Side Navigation Object Reference, Category:Quest UI Category Interface View Model Object Reference=@n7) -> (then:Exec) {is=title K2Node_CallDelegate}
   n8 = OnSideNavigationOnScrollBarVisibilityChanged!() -> (OutputDelegate, then, NewVisibility:ESlateVisibility Enum=Visible) {is=title K2Node_CustomEvent}
   > SetCollapsedMinWidth!(execute, self, NewScrollbarVisibility:ESlateVisibility Enum=@n8.NewVisibility) -> (then)
   > Branch(Condition=Get IsCollapsed() -> (IsCollapsed:bool))
   > SetSizeTransition!(execute, self, Value:double=1) -> (then)
   Event OnAddedToFocusPath!() -> (OutputDelegate, then, InFocusEvent:Focus Event Structure) {is=title K2Node_Event}
   n9 = IsInputType!(InputType:ECommonInputType Enum=Gamepad, IsNot:bool=false) -> (Result:bool)
   > Branch(Condition=@n9)
   > Do Once(Reset=@n10.then)
   > DelayUntilNextTick!(execute) -> (then)
   > SetFocus(self=GetSelectedButton!() -> (ButtonInstance:Common Button Base Object Reference))
   > HandleCollapseRequest(ShouldCollapse=false, IsImmediate=false)
   Event OnRemovedFromFocusPath!() -> (OutputDelegate, then, InFocusEvent:Focus Event Structure) {is=title K2Node_Event}
   > n10 = Branch(Condition=@n9)
   n11 = Event SetIsCollapsed!() -> (OutputDelegate, then, IsCollapsed:bool, SkipAnimation:bool) {is=title K2Node_Event}
   > HandleCollapseRequest(ShouldCollapse=@n11.IsCollapsed, IsImmediate=@n11.SkipAnimation)
   > PlayTransition(IsCollapsed=@n11.IsCollapsed, SkipAnimation=@n11.SkipAnimation)
   Event OnInitialized()
   > Bind Event to On Scroll Bar Visibility Changed(
    self:Scroll Box Object Reference=Get ScrollBox_Tabs!(self:WBP UIKit Tab Group Base Object Reference=Get SideNavigation()) -> (ScrollBox_Tabs:Scroll Box Object Reference)
    Delegate=@n8
   ) {is=title K2Node_AddDelegate}
   > SetCompactUIStyle() {is=title K2Node_CallFunction}
   > SetCollapsedMinWidth(NewScrollbarVisibility=Collapsed)
   n12 = Get SideNavigation()
   > SetVisibility(
    self=Get TouchCatcher_VisualAttachment()
    InVisibility=Select(
     Option 0=SelfHitTestInvisible
     Option 1=Collapsed
     Index=Get SideNavigationIsCollapsed!(self:WBP UIKit Side Navigation Base Object Reference=@n12) -> (SideNavigationIsCollapsed:bool)
    )
   )
  ]}
 GetRemainingAnimTime {nodes [
   GetRemainingAnimTime() {is=title K2Node_FunctionEntry}
   > Return Node!(execute, FunctionIsImplemented:bool=false, RemainingTime:double=0, TotalTime:double=0)
  ]}
 GetRequiresVisibility {nodes [
   GetRequiresVisibility() {is=title K2Node_FunctionEntry}
   n1 = Get CategoryViewModels() -> (CategoryViewModels:Array of Quest UI Category Interface View Model Object References)
   > Return Node!(
    execute
    RequiresVisibility:bool="integer > integer"(A=Length(TargetArray:Array of Quest UI Category Interface View Model Object References=@n1), B=1)
   )
  ]}
 GetSelectedButton {nodes [
   GetSelectedButton() {is=title K2Node_FunctionEntry}
   n1 = Get SideNavigation()
   > Return Node!(
    execute
    ButtonInstance:Common Button Base Object Reference=GetTabButtonBaseByID(self=@n1, TabNameID=GetSelectedTabId!(self:Common Tab List Widget Base Object Reference=@n1) -> (ReturnValue:Name=None))
   )
  ]}
 GetSelectedCategory {nodes [
   GetSelectedCategory() {is=title K2Node_FunctionEntry}
   n1 = Get SideNavigation()
   > n2 = Cast To QuestUICategoryInterfaceVM(Object=GetTabViewModel(self=@n1, TabID=GetSelectedTabId(self=@n1) -> (ReturnValue=None)))
   > Return Node!(execute, SelectedCategory:Quest UI Category Interface View Model Object Reference=@n2)
   Return Node(execute=@n2.CastFailed)
  ]}
 GetTabID {nodes [
   n1 = GetTabID!() -> (then, CategoryInterfaceVM:Quest UI Category Interface View Model Object Reference) {is=title K2Node_FunctionEntry}
   > n2 = Is Valid(InputObject=@n1)
   Return Node!(execute=@n2.Is Not Valid, TabID:Name=InvalidTab)
   > Return Node(TabID=String To Name(InString=GetAnalyticsID!(self:Quest UI Base Content View Model Object Reference=@n1) -> (ReturnValue:String)))
  ]}
 HandleCollapseRequest {nodes [
   n1 = HandleCollapseRequest!() -> (then, ShouldCollapse:bool=~, IsImmediate:bool=~) {is=title K2Node_FunctionEntry}
   > n2 = Sequence()
   > Set LocalShouldCollapse(LocalShouldCollapse:bool=@n1) -> (Output_Get:bool)
   n3 = Get IsCollapsed()
   n4 = Get LocalShouldCollapse() -> (LocalShouldCollapse:bool)
   Branch(
    execute=@n2.then_1
    Condition=OR Boolean(A="Not Equal (Boolean)"(A=@n4, B=@n3) -> (ReturnValue=~), B=Get IsImmediate() -> (IsImmediate:bool))
   )
   > n5 = Set IsCollapsed(IsCollapsed:bool=Get LocalShouldCollapse()) -> (Output_Get:bool)
   > n6 = Branch(Condition=@n5)
   n7 = Get SideNavigation()
   n8 = Get IsImmediate()
   > SetIsCollapsed!(execute, self:WBP UIKit Side Navigation Base Object Reference=@n7, IsCollapsed:bool=true, SkipAnimation:bool=@n8) -> (then)
   SetIsCollapsed(execute=@n6.else, self=@n7, IsCollapsed=false, SkipAnimation=@n8)
  ]}
 IsInFrontend {nodes [
   IsInFrontend() {is=title K2Node_FunctionEntry}
   > n1 = Cast To FortPlayerControllerFrontEnd!(execute, Object=GetOwningPlayer()) -> (then, CastFailed, AsFort Player Controller Front End:Fort Player Controller Front End Object Reference) {is=title K2Node_DynamicCast}
   > Return Node!(execute, InFrontend:bool=true)
   Return Node(execute=@n1.CastFailed, InFrontend=false)
  ]}
 OnCategoryClicked {nodes [OnCategoryClicked!() -> (then, ClickedCategory:Quest UI Category Interface View Model Object Reference) {is=title K2Node_FunctionEntry}]}
 OnCategorySelected {nodes [OnCategorySelected!() -> (then, Category:Quest UI Category Interface View Model Object Reference) {is=title K2Node_FunctionEntry}]}
 OnCollapsedStateChanged {nodes [OnCollapsedStateChanged!() -> (then, IsCollapsed:bool=~) {is=title K2Node_FunctionEntry}]}
 OnRequiresVisibilityChanged {nodes [OnRequiresVisibilityChanged!() -> (then, RequiresVisibility:bool=~) {is=title K2Node_FunctionEntry}]}
 On_MouseCatcher_MouseButtonDown {nodes [
   On_MouseCatcher_MouseButtonDown!() -> (then, MyGeometry:Geometry Structure, MouseEvent:"Pointer Event Structure (by ref)") {is=title K2Node_FunctionEntry}
   > SetIsCollapsed(self:Self Object Reference, IsCollapsed=true, SkipAnimation=false)
   > Return Node!(execute, ReturnValue:Event Reply Structure=Handled())
  ]}
 PlayTransition {nodes [
   n1 = PlayTransition!() -> (then, IsCollapsed:bool=~, SkipAnimation:bool=~) {is=title K2Node_FunctionEntry}
   > n2 = Branch(Condition=@n1.SkipAnimation)
   n3 = Get Anim_Collapse() -> (Anim_Collapse:Widget Animation Object Reference)
   n4 = Select(Option 0:EUMGSequencePlayMode Enum=Reverse, Option 1:EUMGSequencePlayMode Enum, Index=@n1) -> (ReturnValue:EUMGSequencePlayMode Enum)
   > PlayAnimation(InAnimation=@n3, StartAtTime=GetEndTime(self=@n3), PlayMode=@n4)
   n5 = Branch(execute=@n2.else, Condition=IsAnimationPlaying(InAnimation=Get Anim_Collapse()))
   n6 = Get Anim_Collapse()
   > n7 = PauseAnimation!(execute, self:User Widget Object Reference, InAnimation:Widget Animation Object Reference=@n6) -> (then, ReturnValue:float)
   n8 = GetEndTime(self=@n6)
   n9 = float / float(A=1, B=0.33)
   > PlayAnimation(InAnimation=@n6, StartAtTime=SelectFloat(A=@n7, B=float - float(A=@n8, B=@n7), bPickA=@n1), PlayMode=@n4, PlaybackSpeed=@n9)
   PlayAnimation(execute=@n5.else, InAnimation=Get Anim_Collapse(), PlayMode=@n4, PlaybackSpeed=@n9)
  ]}
 SetClearBadgesVisibility {nodes [
   n1 = SetClearBadgesVisibility!() -> (then, Visibility:ESlateVisibility Enum=Visible) {is=title K2Node_FunctionEntry}
   > SetVisibility(InVisibility=@n1)
  ]}
 SetCollapsedMinWidth {nodes [
   n1 = SetCollapsedMinWidth!() -> (then, NewScrollbarVisibility:ESlateVisibility Enum=Visible) {is=title K2Node_FunctionEntry}
   n2 = Select!(
    Visible:bool=true
    Collapsed:bool=false
    Hidden:bool=true
    HitTestInvisible:bool=true
    SelfHitTestInvisible:bool=true
    Index:ESlateVisibility Enum=@n1
   ) -> (ReturnValue:bool)
   > Set CollapsedMinWidth(CollapsedMinWidth:double=SelectFloat(A=SelectFloat(A=132, B=104, bPickA=@n2), B=64, bPickA=IsUsingCompactUI())) -> (Output_Get:double)
  ]}
 SetCompactUIStyle {nodes [
   SetCompactUIStyle() {is=title K2Node_FunctionEntry}
   > Branch(Condition=IsUsingCompactUI())
   n1 = Get SideNavigation()
   > Set!(execute, MinButtonSize_X:int=80, MinButtonSize_Y:int=80, self:WBP UIKit Tab Group Base Object Reference=@n1) -> (then, Output_Get:Int Point Structure)
   > Set UseMinButtonSize!(execute, UseMinButtonSize:bool=true, self:WBP UIKit Tab Group Base Object Reference=@n1) -> (then, Output_Get:bool)
   > Set IconSize!(execute, IconSize:Vector 2D Structure=(X=72,Y=72), self:WBP UIKit Tab Group Base Object Reference=@n1) -> (then, Output_Get:Vector 2D Structure)
   > Set BackplateCornerRadius!(execute, BackplateCornerRadius:E_UI_BackplateCornerRadius Enum=NewEnumerator0, self:WBP UIKit Tab Group Base Object Reference=@n1) -> (then, Output_Get:E_UI_BackplateCornerRadius Enum=NewEnumerator0)
   > Set!(
    execute
    ScrollbarPadding_Left:float=0
    ScrollbarPadding_Top:float=16
    ScrollbarPadding_Right:float=16
    ScrollbarPadding_Bottom:float=16
    self:Scroll Box Object Reference=Get ScrollBox_Tabs(self=@n1)
   ) -> (then, Output_Get:Margin Structure)
  ]}
 SetSizeTransition {nodes [
   n1 = SetSizeTransition!() -> (then, Value:double=~) {is=title K2Node_FunctionEntry}
   > SetMaxDesiredWidth!(
    execute
    self:Size Box Object Reference=Get Root() -> (Root:Size Box Object Reference)
    InMaxDesiredWidth:float=Ease!(
     Function:EEasingFunc Enum=Linear
     Alpha:double=@n1
     A:double=Get!() -> (CachedContentSize_X:double, CachedContentSize_Y:double)
     B:double=Get CollapsedMinWidth() -> (CollapsedMinWidth:double)
    ) -> (Result:double=~) {is=title K2Node_EaseFunction}
   ) -> (then)
  ]}
 VM_InitializeSideNav {nodes [
   VM_InitializeSideNav!() -> (then, CategoryInterfaceVMs:Array of Quest UI Category Interface View Model Object References) {is=title K2Node_FunctionEntry}
   > Clear(TargetArray:Array of Quest UI Category Interface View Model Object References=Get CategoryViewModels())
   > RemoveTabs(self:WBP UIKit Tab Group Base Object Reference=Get SideNavigation())
   > n1 = For Each Loop(Array:Array of Quest UI Category Interface View Model Object References=Get CategoryInterfaceVMs() -> (CategoryInterfaceVMs:Array of Quest UI Category Interface View Model Object References)) -> (Array Element:Quest UI Category Interface View Model Object Reference)
   > Branch(Condition=GetShouldCreateModalTab!(self:Quest UI Category Interface View Model Object Reference=@n1) -> (ReturnValue:bool))
   > Add(TargetArray=Get CategoryViewModels(), NewItem=@n1)
   n2 = For Each Loop(Exec=@n1.Completed, Array=Get CategoryViewModels())
   n3 = Sequence(execute=@n2.Completed)
   > SelectTabByIndex!(execute, self:UIKit Tab Group Object Reference=Get SideNavigation(), Index:int=0, bSuppressClickFeedback:bool=true) -> (then, ReturnValue:bool)
   Call On Category Clicked(
    execute=@n3.then_1
    ClickedCategory="Get (a copy)"(Array:Array of Quest UI Category Interface View Model Object References=Get CategoryViewModels()) -> (Output:Quest UI Category Interface View Model Object Reference)
   )
   > Call On Requires Visibility Changed!(
    execute:Exec
    self:WBP Quest Modal Side Navigation Object Reference
    RequiresVisibility:bool=GetRequiresVisibility!() -> (RequiresVisibility:bool) {is=title K2Node_CallFunction}
   ) -> (then:Exec) {is=title K2Node_CallDelegate}
   AddSideNavigationTab!(
    execute=@n2.LoopBody
    self:WBP UIKit Side Navigation Base Object Reference=Get SideNavigation()
    TabViewModel:MVVM Base Viewmodel Object Reference=@n2
    TabID:Name=GetTabID!(CategoryInterfaceVM:Quest UI Category Interface View Model Object Reference=@n2) -> (TabID:Name=None) {is=title K2Node_CallFunction}
    Index:int=-1
   ) -> (then)
  ]}
]
variables [
 CachedContentSize {type=Vector2D, default=<370 0>}
 CategoryViewModels {type="array<QuestUICategoryInterfaceVM>", default []}
 CollapsedMinWidth {type=number, default=64}
 IsCollapsed {type=boolean}
 OnCategoryClicked {}
 OnCategorySelected {}
 OnCollapsedStateChanged {}
 OnRequiresVisibilityChanged {}
 SelectedCategory {type=QuestUICategoryInterfaceVM, default=null}
 SizeTransition {type=number}
 TrackerVM {type=QuestUICategoryInterfaceVM, default=null}
]
widgets {
 info {inheritedWidgetCount=0, parentClass=&UserWidget, rootWidgetClass=&Overlay, widgetCount=6}
 widgets [
  Overlay_0 {widgetClassPath=&Overlay, widget=Overlay_0}
  VisibilityBorder_Touch {
   widgetClassPath=&CommonHardwareVisibilityBorder
   parent=Overlay_0
   widget=VisibilityBorder_Touch
   properties {
    horizontalAlignment=HAlign_Center
    verticalAlignment=VAlign_Center
    visibilityQuery {
     autoDescription=" ANY( Input.Touch )"
     tokenStreamVersion=0
     userDescription=Touch Only
     queryTokenStream [0,1,1,1,0]
     tagDictionary [{tagName=Input.Touch}]
    }
   }
   slot {class=OverlaySlot}
  }
  TouchCatcher_VisualAttachment {
   widgetClassPath=&CommonVisualAttachment
   bIsVariable=true
   parent=VisibilityBorder_Touch
   widget=TouchCatcher_VisualAttachment
   properties {contentAnchor=<0.5 0.5>}
   slot {class=BorderSlot, properties {horizontalAlignment=HAlign_Center, verticalAlignment=VAlign_Center, padding=<ltrb 0 0 0 0>}}
  }
  TouchCatcher {
   widgetClassPath=&Image
   bIsVariable=true
   parent=TouchCatcher_VisualAttachment
   widget=TouchCatcher
   properties {
    brush {
     bIsDynamicallyLoaded=false
     drawAs=Image
     imageType=NoImage
     mirroring=NoMirror
     resourceName=None
     resourceObject=null
     tiling=NoTile
     imageSize=<8000 8000>
     margin=<ltrb 0 0 0 0>
     outlineSettings {
      bUseBrushTransparency=false
      roundingType=HalfHeightRadius
      width=0
      color {colorUseRule=UseColor_Specified, specifiedColor=<rgba 0 0 0 0>}
      cornerRadii=<0 0 0 0>
     }
     tintColor {colorUseRule=UseColor_Specified, specifiedColor=<rgba 0 0 0 0>}
     uVRegion {bIsValid=false, max=<0 0>, min=<0 0>}
    }
    colorAndOpacity=<rgba 0 0 0 0>
   }
   slot {class=SizeBoxSlot}
  }
  Root {widgetClassPath=&SizeBox, bIsVariable=true, parent=Overlay_0, widget=Root, slot {class=OverlaySlot}}
  SideNavigation {
   widgetClassPath=/FortUILibrary/UIKit/Tabs/WBP_UIKit_SideNavigation
   bIsVariable=true
   parent=Root
   widget=SideNavigation
   properties {
    buttonClass=&/WBP_QuestCategory_SideNavigationButton.WBP_QuestCategory_SideNavigationButton_C
    buttonHorizontalAlignment=HAlign_Left
    dividerClass=/QuestUI/QuestScreen/WBP_QuestsTabDivider.WBP_QuestsTabDivider_C
    dividerSpacing=4
    enableCollapsedState=true
    nextPromptHorizontalAlignment=HAlign_Right
    prevPromptHorizontalAlignment=HAlign_Left
    scrollBarVisibility=Visible
    tabSpacing=8
    usePromptBackplate=false
    iconSize=<40 40>
   }
   slot {class=SizeBoxSlot}
  }
 ]
}
