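# Daily bonus goal list entry for the quest screen, showing a punch card's objective text beside a row of reward widgets.
# The list item object becomes the daily punch card view model.
# PopulateQuests rebuilds the reward row from quest view models, tracking the first quest or the last completed one as current.
# UpdateText then shows that quest's name, optionally with current- or last-stage progress per the card's display mode.
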WidgetBlueprint /QuestUI/QuestScreen/Modal/Goals/WBP_DailyBonusGoal
graphs [
 EventGraph {nodes [
   n1 = Event OnListItemObjectSet()
   > n2 = Cast To QuestUIDailyPunchCardVM!(execute, Object=@n1.ListItemObject) -> (then, CastFailed, AsQuest UI Daily Punch Card View Model:Quest UI Daily Punch Card View Model Object Reference) {is=title K2Node_DynamicCast}
   > Set QuestUIDailyPunchCardVM(QuestUIDailyPunchCardVM:Quest UI Daily Punch Card View Model Object Reference=@n2) -> (Output_Get:Quest UI Daily Punch Card View Model Object Reference)
  ]}
 PopulateQuests {nodes [
   n1 = PopulateQuests!() -> (then, Rewards:Array of Quest UI Base Content View Model Object References) {is=title K2Node_FunctionEntry}
   > ClearChildren(self=Get QuestContainer() -> (QuestContainer:Horizontal Box Object Reference))
   > n2 = For Each Loop(Array:Array of Quest UI Base Content View Model Object References=@n1) -> (Array Element:Quest UI Base Content View Model Object Reference)
   UpdateText!(
    execute=@n2.Completed
    self
    QuestVM:Quest UI Quest View Model Object Reference=Get CurrentQuestVM() -> (CurrentQuestVM:Quest UI Quest View Model Object Reference)
   ) -> (then)
   n3 = Cast To QuestUIQuestVM!(execute=@n2.LoopBody, Object=@n2) -> (then, CastFailed, AsQuest UI Quest View Model:Quest UI Quest View Model Object Reference) {is=title K2Node_DynamicCast}
   > n4 = Branch(Condition="Equal (Integer)"(A=@n2.Array Index))
   > n5 = Set CurrentQuestVM(CurrentQuestVM:Quest UI Quest View Model Object Reference=@n3) -> (Output_Get:Quest UI Quest View Model Object Reference)
   n6 = Construct Object from Class!(
    execute=[@n5.then, @n4.else]
    Class=&/WBP_DailyBonusGoalReward.WBP_DailyBonusGoalReward_C
    self
    QuestUIQuestVM:Quest UI Quest View Model Object Reference
   ) -> (then, ReturnValue:WBP Daily Bonus Goal Reward Object Reference)
   > AddChild(self=Get QuestContainer(), Content=@n6)
   > Set QuestUIQuestVM!(execute, QuestUIQuestVM:Quest UI Quest View Model Object Reference=@n3, self:WBP Daily Bonus Goal Reward Object Reference=@n6) -> (then, Output_Get:Quest UI Quest View Model Object Reference)
   n7 = Get AchievedCount!(self:Quest UI Quest View Model Object Reference=@n3) -> (AchievedCount:int=0)
   n8 = Get RequiredCount!(self:Quest UI Quest View Model Object Reference=@n3) -> (RequiredCount:int=0)
   > Branch(Condition=AND Boolean(A="integer >= integer"(A=@n7, B=@n8), B="integer > integer"(A=@n8)))
   > Set CurrentQuestVM(CurrentQuestVM=@n3)
  ]}
 UpdateText {nodes [
   n1 = UpdateText!() -> (then, QuestVM:Quest UI Quest View Model Object Reference) {is=title K2Node_FunctionEntry}
   > SetVisibility(
    self=Get Rich_ObjectiveQuestText() -> (Rich_ObjectiveQuestText:Rich Text Block Object Reference)
    InVisibility=Select(Option 0=Hidden, Option 1=SelfHitTestInvisible, Index=IsValid(Object=@n1))
   )
   > Get!(execute:Exec) -> (then:Exec, else:Exec, QuestVM:Quest UI Quest View Model Object Reference)
   n2 = Get QuestName!(self:Quest UI Quest View Model Object Reference=Get QuestVM() -> (QuestVM:Quest UI Quest View Model Object Reference)) -> (QuestName:Text)
   n3 = Get OrderedContentVMs!(self:Quest UI Punch Card View Model Object Reference=Get QuestUIDailyPunchCardVM() -> (QuestUIDailyPunchCardVM:Quest UI Daily Punch Card View Model Object Reference)) -> (OrderedContentVMs:Array of Quest UI Base Content View Model Object References)
   n4 = Cast To QuestUIQuestVM!(Object="Get (a copy)"(
     Array:Array of Quest UI Base Content View Model Object References=@n3
     Dimension 1=Last Index!(TargetArray:Array of Quest UI Base Content View Model Object References=@n3) -> (ReturnValue=0)
    ) -> (Output:Quest UI Base Content View Model Object Reference)) -> (AsQuest UI Quest View Model, bSuccess:bool=~)
   n5 = Get RequiredCount(self=@n4) -> (RequiredCount=0)
   n6 = Get AchievedCount(self=@n4) -> (AchievedCount=0)
   n7 = Get QuestVM()
   n8 = Get AchievedCount(self=@n7) -> (AchievedCount=0)
   n9 = Get RequiredCount(self=@n7) -> (RequiredCount=0)
   n10 = Get QuestName(self=@n7)
   n11 = Get QuestUIDailyPunchCardVM()
   n12 = Get QuestVM()
   > SetText(
    self=Get Rich_ObjectiveQuestText()
    InText=Select!(
     NotDisplay:Text=Get QuestName(self=@n12)
     DisplayLastStage:Text=Format Text!(Format="LOCTEXT(\"{body}: {Progress}/{Total}\")", body:Text=@n2, Progress:int=@n6, Total:int=@n5) -> (Result)
     DisplayCurrentStage:Text=Format Text(Format="LOCTEXT(\"{body}: {Progress}/{Total}\")", body=@n10, Progress=@n8, Total=@n9)
     Index:EGoalCardProgressDisplayMode Enum=Get ProgressDisplayMode!(self:Quest UI Punch Card View Model Object Reference=@n11) -> (ProgressDisplayMode:EGoalCardProgressDisplayMode Enum=NotDisplay)
    ) -> (ReturnValue:Text)
   )
  ]}
]
variables []
widgets {
 info {inheritedWidgetCount=0, parentClass=&CommonUserWidget, rootWidgetClass=&GridPanel, widgetCount=6}
 widgets [
  GridPanel_44 {widgetClassPath=&GridPanel, widget=GridPanel_44, properties {columnFill [1,0]}}
  Rich_ObjectiveQuestText {
   widgetClassPath=/Script/UMG.RichTextBlock
   bIsVariable=true
   parent=GridPanel_44
   widget=Rich_ObjectiveQuestText
   properties {
    autoWrapText=true
    text=TextBlock_Rich
    textStyleSet=/Game/Athena/HUD/QuestIcons/DT_RichText_QuestTile
    decoratorClasses [/Game/Athena/HUD/QuestIcons/BP_QuestTile_IconDecorator.BP_QuestTile_IconDecorator_C]
   }
   slot {class=GridSlot, properties {verticalAlignment=VAlign_Center}}
  }
  QuestContainer {
   widgetClassPath=&HorizontalBox
   bIsVariable=true
   parent=GridPanel_44
   widget=QuestContainer
   slot {class=GridSlot, properties {column=1, horizontalAlignment=HAlign_Right, padding=<ltrb 16 0 0 0>}}
  }
  WBP_DailyBonusGoalReward {
   widgetClassPath=&/WBP_DailyBonusGoalReward
   bIsVariable=true
   parent=QuestContainer
   widget=WBP_DailyBonusGoalReward
   slot {class=HorizontalBoxSlot, properties {verticalAlignment=VAlign_Center}}
  }
  WBP_DailyBonusGoalReward_1 {
   widgetClassPath=&WBP_DailyBonusGoalReward
   bIsVariable=true
   parent=QuestContainer
   widget=WBP_DailyBonusGoalReward_1
   slot {class=HorizontalBoxSlot, properties {verticalAlignment=VAlign_Center}}
  }
  WBP_DailyBonusGoalReward_2 {
   widgetClassPath=&WBP_DailyBonusGoalReward
   bIsVariable=true
   parent=QuestContainer
   widget=WBP_DailyBonusGoalReward_2
   slot {class=HorizontalBoxSlot, properties {verticalAlignment=VAlign_Center}}
  }
 ]
}
