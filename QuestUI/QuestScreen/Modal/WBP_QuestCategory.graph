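# Quest category list entry for the quest screen modal, showing an uppercase header above a wrapping description, each collapsing when empty.
# When a list item is set, it unwraps the item, casts it to a QuestUIHeaderVM and stores that view model.
# Separate functions set both texts and switch their colors between dimmed and brighter for the content-enabled state; none are called from the event graph.
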WidgetBlueprint /QuestUI/QuestScreen/Modal/WBP_QuestCategory
graphs [
 EventGraph {nodes [
   n1 = Event OnListItemObjectSet()
   > n2 = Cast To QuestUIHeaderVM!(execute, Object=UnwrapItem!(Item:Object Reference=@n1.ListItemObject) -> (ReturnValue:Object Reference)) -> (then, CastFailed, AsQuest UI Header View Model:Quest UI Header View Model Object Reference) {is=title K2Node_DynamicCast}
   > Set QuestUIHeaderVM(QuestUIHeaderVM:Quest UI Header View Model Object Reference=@n2) -> (Output_Get:Quest UI Header View Model Object Reference)
  ]}
 IsContendEnabled {nodes [
   n1 = IsContendEnabled!() -> (then, bEnabled:bool=~) {is=title K2Node_FunctionEntry}
   > SetColorAndOpacity(
    self=Get Header() -> (Header:Common Text Object Reference)
    InColorAndOpacity=Select!(
     Option 0_SpecifiedColor:Linear Color Structure=(R=0.5,G=0.5,B=0.5,A=0.8)
     Option 0_ColorUseRule:ESlateColorStylingMode Enum
     Option 1_SpecifiedColor:Linear Color Structure=(R=0.9,G=0.9,B=0.9,A=1)
     Option 1_ColorUseRule:ESlateColorStylingMode Enum
     Index=@n1
    ) -> (ReturnValue:Slate Color Structure)
   )
   > SetColorAndOpacity(
    self=Get Description() -> (Description:Common Text Object Reference)
    InColorAndOpacity=Select(Option 0_SpecifiedColor=(R=0.3115,G=0.3115,B=0.3115,A=0.8), Option 1_SpecifiedColor=(R=0.5029,G=0.5029,B=0.5029,A=0.8), Index=@n1)
   )
  ]}
 SetDescription {nodes [
   n1 = SetDescription!() -> (then, InText:Text) {is=title K2Node_FunctionEntry}
   > "SetText (Text)"(self=Get Description(), InText=@n1)
  ]}
 SetHeader {nodes [
   n1 = SetHeader!() -> (then, InText:Text) {is=title K2Node_FunctionEntry}
   > "SetText (Text)"(self=Get Header(), InText=@n1)
  ]}
]
variables []
widgets {
 info {inheritedWidgetCount=0, parentClass=&CommonUserWidget, rootWidgetClass=&GridPanel, widgetCount=3}
 widgets [
  Root {widgetClassPath=&GridPanel, widget=Root, properties {columnFill [1]}}
  Header {
   widgetClassPath=&CommonTextBlock
   bIsVariable=true
   parent=Root
   widget=Header
   properties {
    autoWrapText=true
    bAutoCollapseWithEmptyText=true
    text=CH6 S1 E1: Brothers
    textTransformPolicy=ToUpper
    colorAndOpacity {colorUseRule=UseColor_Specified, specifiedColor=<rgba 0.9 0.9 0.9 0.9>}
    font {
     bForceMonospaced=false
     bMaterialIsStencil=false
     fontMaterial=null
     fontObject=&HeadingNow
     letterSpacing=0
     monospacedWidth=1
     size=22.5
     skewAmount=0
     typefaceFontName=76 Bold
     outlineSettings {
      bApplyOutlineToDropShadows=false
      bMiteredCorners=false
      bSeparateFillAlpha=false
      outlineMaterial=null
      outlineSize=0
      outlineColor=<rgba 0 0 0 1>
     }
    }
    shadowOffset=<0 0>
   }
   slot {class=GridSlot, properties {padding=<ltrb 0 8 0 0>}}
  }
  Description {
   widgetClassPath=&CommonTextBlock
   bIsVariable=true
   parent=Root
   widget=Description
   properties {
    autoWrapText=true
    bAutoCollapseWithEmptyText=true
    text="Magic is back! Set up the conflict between brothers, Ronin and Mask Maker. Confront a corrupted Yokai Boss."
    colorAndOpacity {colorUseRule=UseColor_Specified, specifiedColor=<rgba 0.4 0.4 0.4 1>}
    font {
     bForceMonospaced=false
     bMaterialIsStencil=false
     fontMaterial=null
     fontObject=&HeadingNow
     letterSpacing=10
     monospacedWidth=1
     size=14.25
     skewAmount=0
     typefaceFontName=74 Regular
     outlineSettings {
      bApplyOutlineToDropShadows=false
      bMiteredCorners=false
      bSeparateFillAlpha=false
      outlineMaterial=null
      outlineSize=0
      outlineColor=<rgba 0 0 0 1>
     }
    }
    shadowOffset=<0 0>
   }
   slot {class=GridSlot, properties {row=1, padding=<ltrb 0 4 0 0>}}
  }
 ]
}
