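# Quest category entry widget for the quest screen modal, showing a category's description and an expiration label.
# It caches the category view model on list assignment and exposes SetDescription and SetExpirationTime.
# SetExpirationTime shows a countdown if over a second remains against MCP UTC time, or custom expiration text when no expiration exists.
# Otherwise it hides the timer, collapsing the whole entry if the category description is also empty.
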WidgetBlueprint /QuestUI/QuestScreen/Modal/WBP_QuestCategoryDescription_Entry
graphs [
 EventGraph {nodes [
   n1 = Event OnListItemObjectSet()
   > n2 = Cast To QuestUICategoryVM!(execute, Object=@n1.ListItemObject) -> (then, CastFailed, AsQuest UI Category View Model:Quest UI Category View Model Object Reference) {is=title K2Node_DynamicCast}
   > Set QuestUICategoryVM(QuestUICategoryVM:Quest UI Category View Model Object Reference=@n2) -> (Output_Get:Quest UI Category View Model Object Reference)
  ]}
 SetDescription {nodes [
   n1 = SetDescription!() -> (then, Description:Text) {is=title K2Node_FunctionEntry}
   > "SetText (Text)"(self=Get Description() -> (Description:Common Text Object Reference), InText=@n1)
  ]}
 SetExpirationTime {nodes [
   n1 = SetExpirationTime!() -> (then, Time:Date Time Structure) {is=title K2Node_FunctionEntry}
   > n2 = Branch(Condition=Get bHasExpirationTime!(self:Quest UI Category View Model Object Reference=Get QuestUICategoryVM() -> (QuestUICategoryVM:Quest UI Category View Model Object Reference)) -> (bHasExpirationTime:bool))
   > n3 = GetMcpRelativeUtcNow!(
    execute
    FortPC:Fort Player Controller Object Reference=GetFirstLocalFortPlayerController!(bIncludeNonSpectators:bool=true, bIncludeSpectators:bool=false) -> (ReturnValue:Fort Player Controller Object Reference)
   ) -> (then, ReturnValue:Date Time Structure)
   > n4 = Branch(Condition="Timespan > Timespan"(
     A:Timespan Structure=DateTime - DateTime!(A:Date Time Structure=@n1, B:Date Time Structure=@n3) -> (ReturnValue:Timespan Structure)
     B:Timespan Structure=MakeTimespan!(Days:int=0, Hours:int=0, Minutes:int=0, Seconds:int=1, Milliseconds:int=0) -> (ReturnValue:Timespan Structure)
    ) -> (ReturnValue:bool=~) {is=title K2Node_PromotableOperator})
   > SetVisibility(self=Get OverlayExpiration() -> (OverlayExpiration:Overlay Object Reference), InVisibility=SelfHitTestInvisible)
   > n5 = SetDateTimeValue!(
    execute
    self:Common Date Time Text Block Object Reference=Get ExpirationTime() -> (ExpirationTime:Athena Challenge Date Time Text Block Object Reference)
    InDateTime:Date Time Structure=@n1
    bShowAsCountdown:bool=true
    InRefreshDelay:float=1
   ) -> (then)
   SetVisibility(execute=[@n10.then, @n5.then], InVisibility=SelfHitTestInvisible)
   SetVisibility(execute=[@n4.else, @n8.then], self=Get OverlayExpiration(), InVisibility=Collapsed)
   n6 = GetCategoryDescription!(self:Quest UI Category View Model Object Reference=Get QuestUICategoryVM()) -> (ReturnValue:Text)
   > SetVisibility(InVisibility=Select(Option 0=SelfHitTestInvisible, Option 1=Collapsed, Index=TextIsEmpty(InText=@n6)))
   n7 = Get CustomExpirationText!(self:Quest UI Category View Model Object Reference=Get QuestUICategoryVM()) -> (CustomExpirationText:Text)
   n8 = Branch(execute=@n2.else, Condition=TextIsEmpty(InText=@n7))
   SetVisibility(execute=@n8.else, self=Get OverlayExpiration(), InVisibility=SelfHitTestInvisible)
   n9 = Get ExpirationTime()
   > SetDateTimeValue(self=@n9, bShowAsCountdown=false, InRefreshDelay=0)
   > n10 = "SetText (Text)"(self=@n9, InText=@n7)
  ]}
]
variables []
widgets {
 info {inheritedWidgetCount=0, parentClass=&CommonUserWidget, rootWidgetClass=&Overlay, widgetCount=6}
 widgets [
  Overlay_0 {widgetClassPath=&Overlay, widget=Overlay_0}
  Root {
   widgetClassPath=&GridPanel
   parent=Overlay_0
   widget=Root
   properties {columnFill [1], rowFill [1]}
   slot {class=OverlaySlot, properties {horizontalAlignment=HAlign_Fill, padding=<ltrb 0 -20 0 0>}}
  }
  Description {
   widgetClassPath=&CommonTextBlock
   bIsVariable=true
   parent=Root
   widget=Description
   properties {
    autoWrapText=true
    bAutoCollapseWithEmptyText=true
    scrollOrientation=Orient_Vertical
    scrollStyle=&TextScrollStyle-Base_C
    text=Two lines of instruction or flavor text describing how to complete the quests in this specific category.
    font {
     bForceMonospaced=false
     bMaterialIsStencil=false
     fontMaterial=null
     fontObject=&HeadingNow
     letterSpacing=0
     monospacedWidth=1
     size=14.25
     skewAmount=0
     typefaceFontName=74 Regular
     outlineSettings {
      bApplyOutlineToDropShadows=false
      bMiteredCorners=false
      bSeparateFillAlpha=false
      outlineMaterial=null
      outlineSize=0
      outlineColor=<rgba 0 0 0 1>
     }
    }
    shadowOffset=<0 0>
   }
   slot {class=GridSlot, properties {layer=1, verticalAlignment=VAlign_Top, padding=<ltrb 0 0 100 14>}}
  }
  OverlayExpiration {
   widgetClassPath=&Overlay
   bIsVariable=true
   parent=Root
   widget=OverlayExpiration
   slot {class=GridSlot, properties {horizontalAlignment=HAlign_Left, layer=1, row=1, verticalAlignment=VAlign_Top, padding=<ltrb 0 0 0 14>}}
  }
  TimeBackground {
   widgetClassPath=&Image
   parent=OverlayExpiration
   widget=TimeBackground
   properties {brush {
     bIsDynamicallyLoaded=false
     drawAs=Image
     imageType=FullColor
     mirroring=NoMirror
     resourceName=None
     resourceObject=/FortniteCoreUI/UI/Quests/MaterialInstances/MI_UI_TimerBackground
     tiling=NoTile
     imageSize=<32 40>
     margin=<ltrb 0 0 0 0>
     outlineSettings {
      bUseBrushTransparency=false
      roundingType=HalfHeightRadius
      width=0
      color {colorUseRule=UseColor_Specified, specifiedColor=<rgba 0 0 0 0>}
      cornerRadii=<0 0 0 0>
     }
     tintColor {colorUseRule=UseColor_Specified, specifiedColor=<rgba 1 1 1 1>}
     uVRegion {bIsValid=false, max=<0 0>, min=<0 0>}
    }}
   slot {class=OverlaySlot, properties {horizontalAlignment=HAlign_Fill, verticalAlignment=VAlign_Fill}}
  }
  ExpirationTime {
   widgetClassPath=/Script/FortniteUI.AthenaChallengeDateTimeTextBlock
   bIsVariable=true
   parent=OverlayExpiration
   widget=ExpirationTime
   properties {
    bAllowSeconds=false
    bAutoCollapseWithEmptyText=true
    bIsScrollingEnabled=false
    completionText=EXPIRED
    labelFormat="Ends: {0}"
    text=Text Block
    colorAndOpacity {colorUseRule=UseColor_Specified, specifiedColor=<rgba 0.5776 0.5776 0.5776 1>}
    font {
     bForceMonospaced=false
     bMaterialIsStencil=false
     fontMaterial=null
     fontObject=&HeadingNow
     letterSpacing=0
     monospacedWidth=1
     size=14.25
     skewAmount=0
     typefaceFontName=74 Regular
     outlineSettings {
      bApplyOutlineToDropShadows=false
      bMiteredCorners=false
      bSeparateFillAlpha=false
      outlineMaterial=null
      outlineSize=0
      outlineColor=<rgba 0 0 0 1>
     }
    }
    shadowOffset=<0 0>
   }
   slot {class=OverlaySlot, properties {verticalAlignment=VAlign_Center, padding=<ltrb 16 0 16 0>}}
  }
 ]
}
