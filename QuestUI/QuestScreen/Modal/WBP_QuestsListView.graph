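# Quest screen list widget wrapping a polymorphic list that maps quest view models to tiles, headers, separators and goal entries.
# RefreshQuestList rebuilds items with a trailing spacer, then optionally selects the requested item or a default first selectable, preferring quests when in-zone.
# Scrolling the first or last selectable item into view snaps next tick to the list's edge, with extra bottom conditions in compact UI.
# SelectQuest navigates to a quest, focusing the list for gamepad.
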WidgetBlueprint /QuestUI/QuestScreen/Modal/WBP_QuestsListView
graphs [
 EventGraph {nodes [
   n1 = OnItemScrolledIntoView!() -> (OutputDelegate, then, Item:Object Reference, Widget:User Widget Object Reference) {is=title K2Node_CustomEvent}
   n2 = GetIndexForItem!(
    self:List View Object Reference=Get ListView!(self:Fort Polymorphic List View Object Reference=Get QuestTilesList() -> (QuestTilesList:WBP Quest Poly List Object Reference)) -> (ListView:Common List View Object Reference)
    Item:Object Reference=@n1.Item
   ) -> (ReturnValue:int=0)
   > n3 = Branch(Condition="Equal (Integer)"(A=@n2, B=Get FirstSelectableItem() -> (FirstSelectableItem:int=0)))
   > DelayUntilNextTick!(execute) -> (then)
   n4 = Get ListView(self=Get QuestTilesList())
   > ScrollToTop(self:List View Base Object Reference=@n4)
   n5 = IsUsingCompactUI()
   n6 = NOT Boolean(A=@n5)
   n7 = Get SelectableQuestItems() -> (SelectableQuestItems:Array of Integers)
   n8 = AND Boolean!(
    A=@n5
    B=Contains Item(
     TargetArray:Array of Integers=Get SelectableQuestItems()
     ItemToFind:"Integer (by ref)"=Get LastSelecatableItem() -> (LastSelecatableItem:int=0)
    )
    C:bool="integer >= integer"(A=Length(TargetArray:Array of Integers=@n7), B=3)
   ) -> (ReturnValue)
   n9 = Get LastSelecatableItem() -> (LastSelecatableItem=0)
   Branch(execute=@n3.else, Condition=AND Boolean!(A="Equal (Integer)"(A=@n2, B=@n9), B=OR Boolean(A=@n6, B=@n8)) -> (ReturnValue))
   > DelayUntilNextTick()
   > ScrollToBottom(self:List View Base Object Reference=@n4)
   Event Construct()
   n10 = Get ListView(self=Get QuestTilesList())
   > Assign On Item Scrolled Into View!(execute:Exec, self:List View Object Reference=@n10, Delegate:"Delegate (by ref)"=@n1) -> (then:Exec) {is=title K2Node_AssignDelegate}
   Event Destruct()
   > Unbind Event from On Item Scrolled Into View(self:List View Object Reference=@n10, Delegate=@n1) {is=title K2Node_RemoveDelegate}
  ]}
 FindFirstSelectableElement {nodes [
   FindFirstSelectableElement() {is=title K2Node_FunctionEntry}
   > Set FirstSelectableItem(FirstSelectableItem:int=0) -> (Output_Get:int=0)
   > Set FirstPreferredSelectableItem(FirstPreferredSelectableItem:int=0) -> (Output_Get:int=0)
   n1 = Get QuestTilesList()
   > n2 = For Each Loop with Break!(
    Exec:Exec
    Array:Array of Object References=GetListItems!(self:List View Object Reference=Get ListView(self=@n1)) -> (ReturnValue:Array of Object References)
    Break:Exec=@n4.then
   ) -> (LoopBody:Exec, Array Element:Object Reference, Array Index:int, Completed:Exec)
   > Branch(Condition=CheckItemSelectability!(self:WBP Quest Poly List Object Reference=Get QuestTilesList(), InputObject:Object Reference=@n2) -> (ReturnValue:bool))
   > n3 = Branch(Condition=Get Finded() -> (Finded:bool))
   Cast To QuestUIQuestVM!(execute=[@n3.then, @n5.then], Object=@n2) -> (then, CastFailed, AsQuest UI Quest View Model:Quest UI Quest View Model Object Reference) {is=title K2Node_DynamicCast}
   > n4 = Set FirstPreferredSelectableItem(FirstPreferredSelectableItem=@n2.Array Index) -> (Output_Get=0)
   Set Finded(execute=@n3.else, Finded:bool=true) -> (Output_Get:bool)
   > n5 = Set FirstSelectableItem(FirstSelectableItem=@n2.Array Index) -> (Output_Get=0)
   Branch(execute=@n2.Completed, Condition="Equal (Integer)"(A=Get FirstPreferredSelectableItem() -> (FirstPreferredSelectableItem:int=0)))
   > Set FirstPreferredSelectableItem(FirstPreferredSelectableItem=Get FirstSelectableItem() -> (FirstSelectableItem=0)) -> (Output_Get=0)
  ]}
 FindLastSelectableElement {nodes [
   FindLastSelectableElement() {is=title K2Node_FunctionEntry}
   n1 = Get ListView(self=Get QuestTilesList())
   > Set LastSelecatableItem(LastSelecatableItem:int=Last Index!(TargetArray:Array of Object References=GetListItems(self=@n1)) -> (ReturnValue=0)) -> (Output_Get:int=0)
   n2 = GetListItems(self=Get ListView(self=Get QuestTilesList()))
   > n3 = For Each Loop with Break(Array=@n2, Break=@n6.then)
   n4 = int - int(A=Last Index(TargetArray=@n2) -> (ReturnValue=0), B=@n3.Array Index)
   n5 = Get ListView(self=Get QuestTilesList())
   > Branch(Condition=CheckItemSelectability(self=Get QuestTilesList(), InputObject=GetItemAt!(self:List View Object Reference=@n5, Index:int=@n4) -> (ReturnValue:Object Reference)))
   > n6 = Set LastSelecatableItem(LastSelecatableItem=@n4) -> (Output_Get=0)
  ]}
 FocusListview {nodes [
   FocusListview() {is=title K2Node_FunctionEntry}
   n1 = Get CommonInputSubsystem()
   > Branch(Condition="Equal (Enum)"(A=GetCurrentInputType(self=@n1), B=Gamepad))
   > SetFocus(self=Get ListView(self=Get QuestTilesList()))
  ]}
 RefreshQuestList {nodes [
   n1 = RefreshQuestList!() -> (then, InListItems:Array of Object References, InitialSelection:Object Reference, IsDraggable:bool=~, FocusQuest:bool=true) {is=title K2Node_FunctionEntry}
   > Set StartingSelection(StartingSelection:Object Reference=@n1.InitialSelection) -> (Output_Get:Object Reference)
   > SetIsDraggable!(execute, self:WBP Quest Poly List Object Reference=Get QuestTilesList(), IsDraggable:bool=@n1.IsDraggable) -> (then)
   > Clear(TargetArray:Array of Object References=Get RefreshedQuestList() -> (RefreshedQuestList:Array of Object References))
   n2 = Get ListView(self=Get QuestTilesList())
   > ClearListItems(self=@n2)
   > Clear Selection(self:List View Object Reference=@n2)
   > n3 = Sequence!(execute) -> (then_0, then_1, then_2:Exec)
   > K2Node_Composite_0 = BuildRefreshedQuestList!(Exec:Exec) {is=title K2Node_Composite}
   n4 = Construct Object from Class!(execute=@n3.then_1, Class=&/QuestUISpacerVM.QuestUISpacerVM_C, self, Size:double=80) -> (then, ReturnValue:Quest UISpacer VM Object Reference)
   > Add(TargetArray:Array of Object References=Get RefreshedQuestList(), NewItem:"Object Reference (by ref)"=@n4) {is=title K2Node_CallArrayFunction}
   > SetListItems!(execute, self:Fort Polymorphic View Object Reference=Get QuestTilesList(), InListItems:Array of Object References=Get RefreshedQuestList()) -> (then)
   n5 = Get ListView(self=Get QuestTilesList())
   > RegenerateAllEntries(self:List View Base Object Reference=@n5)
   > Clear(TargetArray:Array of Integers=Get SelectableQuestItems())
   > n6 = For Each Loop(Array:Array of Object References=GetListItems(self=@n5)) -> (Array Element:Object Reference)
   > Branch(Condition=CheckItemSelectability(self=Get QuestTilesList(), InputObject=@n6))
   > Cast To QuestUIQuestVM(Object=@n6)
   > Add Unique(TargetArray:Array of Integers=Get SelectableQuestItems(), NewItem:"Integer (by ref)"=@n6.Array Index)
   Branch(execute=@n3.then_2, Condition=Get FocusQuest() -> (FocusQuest:bool))
   > FindFirstSelectableElement() {is=title K2Node_CallFunction}
   > FindLastSelectableElement() {is=title K2Node_CallFunction}
   n7 = Get StartingSelection() -> (StartingSelection:Object Reference)
   > n8 = Is Valid(InputObject=@n7)
   n9 = Get FortGlobalUIContext() -> (ReturnValue:Fort Global UIContext Object Reference) {is=title K2Node_GetSubsystem}
   SelectIndex!(
    execute=[@n11.else, @n8.Is Not Valid]
    self
    Index:int=Select(
     Option 0:int=Get FirstSelectableItem() -> (FirstSelectableItem=0)
     Option 1:int=Get FirstPreferredSelectableItem() -> (FirstPreferredSelectableItem=0)
     Index=IsInZone!(self:Fort Global UIContext Object Reference=@n9) -> (ReturnValue:bool)
    ) -> (ReturnValue:int)
   ) -> (then)
   n10 = Find Item!(TargetArray:Array of Object References=Get RefreshedQuestList(), ItemToFind:"Object Reference (by ref)"=@n7) -> (ReturnValue)
   n11 = Branch(execute=@n8.Is Valid, Condition="integer >= integer"(A=@n10))
   > SelectIndex(Index=@n10)
  ]}
 BuildRefreshedQuestList {
  path=RefreshQuestList.K2Node_Composite_0.BuildRefreshedQuestList
  nodes [
   Inputs!() -> (Exec:Exec)
   > Append Array!(
    execute
    TargetArray:Array of Object References=Get RefreshedQuestList()
    SourceArray:Array of Object References=Get InListItems() -> (InListItems:Array of Object References)
   ) -> (then)
   Outputs()
  ]
 }
 SelectIndex {nodes [
   n1 = SelectIndex!() -> (then, Index:int) {is=title K2Node_FunctionEntry}
   n2 = Get ListView(self=Get QuestTilesList())
   > SetSelectedIndex!(execute, self:List View Object Reference=@n2, Index:int=@n1) -> (then)
   > ScrollIndexIntoView!(execute, self:List View Object Reference=@n2, Index:int=@n1) -> (then)
  ]}
 SelectQuest {nodes [
   n1 = SelectQuest!() -> (then, QuestToSelect:Quest UI Base Content View Model Object Reference) {is=title K2Node_FunctionEntry}
   n2 = Get ListView(self=Get QuestTilesList())
   > Set Selected Item!(execute, self:List View Object Reference=@n2, Item:Object Reference=@n1) -> (then) {is=title K2Node_CallFunction}
   > FocusListview() {is=title K2Node_CallFunction}
   > Navigate To Item!(
    execute
    self:List View Object Reference=@n2
    Item:Object Reference=Get QuestToSelect() -> (QuestToSelect:Quest UI Base Content View Model Object Reference)
   ) -> (then)
  ]}
 SetIsMoveMode {nodes [
   n1 = SetIsMoveMode!() -> (then, IsMoveMode:bool=~) {is=title K2Node_FunctionEntry}
   > SetIsMoveMode!(execute, self:WBP Quest Poly List Object Reference=Get QuestTilesList(), IsMoveMode:bool=@n1) -> (then) {is=title K2Node_CallFunction}
  ]}
]
variables [
 FirstPreferredSelectableItem {type=integer}
 FirstSelectableItem {type=integer}
 LastSelecatableItem {type=integer}
 List {type="array<MVVMViewModelBase>", default []}
 SelectableQuestItems {type="array<integer>"}
]
widgets {
 info {
  inheritedWidgetCount=0
  parentClass=&CommonUserWidget
  rootWidgetClass=&/WBP_QuestPolyList.WBP_QuestPolyList_C
  widgetCount=1
 }
 widgets [QuestTilesList {
   widgetClassPath=&/WBP_QuestPolyList
   bIsVariable=true
   widget=QuestTilesList
   properties {
    bIsFocusable=true
    desiredFocusWidget {widgetName=None}
    itemClassToEntryData {
     &Class'/Script/QuestUI.QuestUICategoryVM' {
      bIsEntrySelectable=false
      desiredEntryClass=&/WBP_QuestCategoryDescription_Entry.WBP_QuestCategoryDescription_Entry_C
     }
     &Class'/Script/QuestUI.QuestUICompatibilityVM' {
      bIsEntrySelectable=false
      desiredEntryClass=/QuestUI/QuestScreen/QuestTile/WBP_QuestGroupSeparatorTitle_Entry.WBP_QuestGroupSeparatorTitle_Entry_C
     }
     &Class'/Script/QuestUI.QuestUIDailyPunchCardVM' {bIsEntrySelectable=false, desiredEntryClass=/QuestUI/QuestScreen/Modal/Goals/WBP_DailyBonusGoal.WBP_DailyBonusGoal_C}
     &Class'/Script/QuestUI.QuestUIHeaderVM' {bIsEntrySelectable=false, desiredEntryClass=&/WBP_QuestCategory.WBP_QuestCategory_C}
     &Class'/Script/QuestUI.QuestUIHorizontalPunchCardVM' {bIsEntrySelectable=true, desiredEntryClass=/QuestUI/QuestScreen/Modal/Goals/WBP_GoalsProgress_Horizontal.WBP_GoalsProgress_Horizontal_C}
     &Class'/Script/QuestUI.QuestUILockedQuestVM' {
      bIsEntrySelectable=false
      desiredEntryClass=&WBP_QuestGroupSeparatorTitle_Entry_C
     }
     &Class'/Script/QuestUI.QuestUIMessageVM' {
      bIsEntrySelectable=false
      desiredEntryClass=&WBP_QuestGroupSeparatorTitle_Entry_C
     }
     &Class'/Script/QuestUI.QuestUIQuestVM' {bIsEntrySelectable=true, desiredEntryClass=/QuestUI/QuestScreen/QuestTile/WBP_QuestTile.WBP_QuestTile_C}
     /Script/Engine.BlueprintGeneratedClass'/QuestUI/QuestScreen/Debug/DebugQuestUIHeaderVM.DebugQuestUIHeaderVM_C' {
      bIsEntrySelectable=false
      desiredEntryClass=&WBP_QuestGroupSeparatorTitle_Entry_C
     }
     &BlueprintGeneratedClass'/QuestUI/QuestScreen/Modal/Goals/QuestUIGoalHeaderVM.QuestUIGoalHeaderVM_C' {
      bIsEntrySelectable=false
      desiredEntryClass=/QuestUI/QuestScreen/Modal/Goals/WBP_GoalProgressHeader_Vertical.WBP_GoalProgressHeader_Vertical_C
     }
     &BlueprintGeneratedClass'&/MVVM_PunchCardQuest.MVVM_PunchCardQuest_C' {bIsEntrySelectable=true, desiredEntryClass=&WBP_QuestTile_C}
     &BlueprintGeneratedClass'&QuestUISpacerVM_C' {bIsEntrySelectable=false, desiredEntryClass=/QuestUI/QuestScreen/QuestTile/WBP_QuestSpacer_Entry.WBP_QuestSpacer_Entry_C}
     &BlueprintGeneratedClass'&/VM_QuestUIPinnedQuestSeparator.VM_QuestUIPinnedQuestSeparator_C' {bIsEntrySelectable=false, desiredEntryClass=&/WBP_PinnedSeparator.WBP_PinnedSeparator_C}
    }
   }
  }]
}
