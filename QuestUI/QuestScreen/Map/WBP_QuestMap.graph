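# WBP_QuestMap is the quest screen's map widget, derived from QuestUIMap, hosting the inherited Athena map layer container inside a fixed-size box.
# The ShowMapLayerContainer event toggles MapAnimWrapper, the container's parent overlay, between SelfHitTestInvisible and Collapsed.
# BP_GetDesiredInputConfig returns a stored input config allowing all input modes without mouse capture or lock.
# A collapsed custom background and a zero-opacity punch-card overlay image are also present, undriven by this graph.
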WidgetBlueprint /QuestUI/QuestScreen/Map/WBP_QuestMap
graphs [
 BP_GetDesiredInputConfig {nodes [
   Get Desired Input Config() {is=title K2Node_FunctionEntry}
   > Return Node!(execute, ReturnValue:UIInput Config Structure=Get DesiredInputConfig() -> (DesiredInputConfig:UIInput Config Structure))
  ]}
 EventGraph {nodes [
   n1 = Event ShowMapLayerContainer!() -> (OutputDelegate, then, bShowing:bool) {is=title K2Node_Event}
   > n2 = Branch(Condition=@n1.bShowing)
   > SetVisibility(self=Get MapAnimWrapper() -> (MapAnimWrapper:Overlay Object Reference), InVisibility=SelfHitTestInvisible)
   SetVisibility(execute=@n2.else, self=Get MapAnimWrapper(), InVisibility=Collapsed)
   Event PreConstruct()
   Event Construct()
   Event Tick()
  ]}
]
variables [DesiredInputConfig {
  type=UIInputConfig
  default {
   bHideCursorDuringViewportCapture=true
   bIgnoreLookInput=false
   bIgnoreMoveInput=false
   inputMode=All
   mouseCaptureMode=NoCapture
   mouseLockMode=DoNotLock
  }
 }]
widgets {
 info {inheritedWidgetCount=1, parentClass=/Script/QuestUI.QuestUIMap, rootWidgetClass=&Overlay, widgetCount=10}
 widgets [
  Overlay_2 {widgetClassPath=&Overlay, widget=Overlay_2}
  Image_CustomMapBackground {
   widgetClassPath=&Image
   parent=Overlay_2
   widget=Image_CustomMapBackground
   properties {visibility=Collapsed}
   slot {class=OverlaySlot, properties {horizontalAlignment=HAlign_Fill, verticalAlignment=VAlign_Fill}}
  }
  SizeBox_1 {
   widgetClassPath=&SizeBox
   parent=Overlay_2
   widget=SizeBox_1
   properties {bOverride_HeightOverride=true, bOverride_WidthOverride=true, heightOverride=1140, widthOverride=1748}
   slot {class=OverlaySlot, properties {horizontalAlignment=HAlign_Fill, verticalAlignment=VAlign_Fill}}
  }
  Overlay_0 {widgetClassPath=&Overlay, parent=SizeBox_1, widget=Overlay_0, slot {class=SizeBoxSlot}}
  Overlay_20 {
   widgetClassPath=&Overlay
   parent=Overlay_0
   widget=Overlay_20
   slot {class=OverlaySlot, properties {horizontalAlignment=HAlign_Fill, verticalAlignment=VAlign_Fill}}
  }
  Border_0 {
   widgetClassPath=&Border
   parent=Overlay_20
   widget=Border_0
   properties {
    visibility=SelfHitTestInvisible
    background {
     bIsDynamicallyLoaded=false
     drawAs=NoDrawType
     imageType=NoImage
     mirroring=NoMirror
     resourceName=None
     resourceObject=null
     tiling=NoTile
     imageSize=<32 32>
     margin=<ltrb 0 0 0 0>
     outlineSettings {
      bUseBrushTransparency=false
      roundingType=HalfHeightRadius
      width=0
      color {colorUseRule=UseColor_Specified, specifiedColor=<rgba 0 0 0 0>}
      cornerRadii=<0 0 0 0>
     }
     tintColor {colorUseRule=UseColor_Specified, specifiedColor=<rgba 1 1 1 1>}
     uVRegion {bIsValid=false, max=<0 0>, min=<0 0>}
    }
    padding=<ltrb 0 0 0 0>
   }
   slot {class=OverlaySlot, properties {horizontalAlignment=HAlign_Fill, verticalAlignment=VAlign_Fill}}
  }
  MapAnimWrapper {
   widgetClassPath=&Overlay
   bIsVariable=true
   parent=Border_0
   widget=MapAnimWrapper
   slot {class=BorderSlot, properties {padding=<ltrb 0 0 0 0>}}
  }
  MapLayerContainer {
   widgetClassPath=/FortniteCoreUI/HUD/MiniMap/AthenaMapLayerContainer
   bInherited=true
   parent=MapAnimWrapper
   widget=MapLayerContainer
   properties {bConsumePointerOnTouchInteraction=true, enableMapSafeZone=false}
   slot {class=OverlaySlot, properties {horizontalAlignment=HAlign_Fill, verticalAlignment=VAlign_Fill}}
  }
  SafeZone_1 {
   widgetClassPath=&SafeZone
   bIsVariable=true
   parent=Overlay_20
   widget=SafeZone_1
   properties {padBottom=false, padLeft=false, padRight=false}
   slot {class=OverlaySlot, properties {horizontalAlignment=HAlign_Fill, verticalAlignment=VAlign_Fill, padding=<ltrb 40 32 40 0>}}
  }
  Image_Punchards_ {
   widgetClassPath=&Image
   bIsVariable=true
   parent=Overlay_0
   widget=Image_Punchards_
   properties {
    renderOpacity=0
    visibility=HitTestInvisible
    brush {
     bIsDynamicallyLoaded=false
     drawAs=Image
     imageType=NoImage
     mirroring=NoMirror
     resourceName=None
     resourceObject=/FortniteCoreUI/Athena/HUD/Quests/Art/PunchCard/M_Ui_Map_OffScreenOverlay_Inst
     tiling=NoTile
     imageSize=<32 32>
     margin=<ltrb 0 0 0 0>
     outlineSettings {
      bUseBrushTransparency=false
      roundingType=HalfHeightRadius
      width=0
      color {colorUseRule=UseColor_Specified, specifiedColor=<rgba 0 0 0 0>}
      cornerRadii=<0 0 0 0>
     }
     tintColor {colorUseRule=UseColor_Specified, specifiedColor=<rgba 1 1 1 1>}
     uVRegion {bIsValid=false, max=<0 0>, min=<0 0>}
    }
   }
   slot {class=OverlaySlot, properties {horizontalAlignment=HAlign_Fill, verticalAlignment=VAlign_Fill, padding=<ltrb -512 -512 -512 -512>}}
  }
 ]
}
