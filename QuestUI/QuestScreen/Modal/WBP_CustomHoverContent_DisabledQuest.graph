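# Hover content widget telling the player that a quest cannot be progressed on the selected island.
# Setting the quest link ID shows the select button only for a non-empty ID in the frontend; the treatment type labels it "Select Island" or "View Genre".
# Clicking broadcasts link ID and type through the playspace Gameplay Event Router.
# DoCustomNavigation fires OnSelectIslandNavLeft on left navigation; Destruct clears both dispatchers.
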WidgetBlueprint /QuestUI/QuestScreen/Modal/WBP_CustomHoverContent_DisabledQuest
graphs [
 DoCustomNavigation {nodes [
   n1 = DoCustomNavigation!() -> (then, Navigation:EUINavigation Enum) {is=title K2Node_FunctionEntry}
   > Switch on EUINavigation!(execute, Selection:EUINavigation Enum=@n1) -> (Left:Exec, Right:Exec, Up:Exec, Down:Exec, Next:Exec, Previous:Exec, Invalid:Exec) {is=title K2Node_SwitchEnum}
   > Call On Select Island Nav Left!(execute:Exec, self:WBP Custom Hover Content Disabled Quest Object Reference) -> (then:Exec) {is=title K2Node_CallDelegate}
   > Return Node!(execute, ReturnValue:Widget Object Reference)
  ]}
 EventGraph {nodes [
   "On Clicked (Button_SelectIsland)"!() -> (OutputDelegate, then, Button:Common Button Base Object Reference) {is=title K2Node_ComponentBoundEvent}
   > n1 = GetEventRouter!(
    execute
    Scope:Object Class Reference=/Script/FortniteGame.FortPlayspace
    ContextActor:Actor Object Reference=GetOwningPlayer()
    None:%=Get GameplayEventRouterSubsystem() -> (ReturnValue:Gameplay Event Router Subsystem Object Reference) {is=title K2Node_GetSubsystem}
   ) -> (then, ReturnValue:Gameplay Event Router Component Object Reference)
   > Broadcast Event!(
    execute
    self:Gameplay Event Router Component Object Reference=@n1
    EventData:FSelect Island Event Data Structure=Make FSelect Island Event Data!(
     LinkID_2_#1:Online Link Id Structure=Get OnlineLinkID() -> (OnlineLinkID:Online Link Id Structure)
     LinkType_5_#2:EDiscoveryLinkTreatmentType Enum=Get DiscoveryLinkTreatmentType() -> (DiscoveryLinkTreatmentType:EDiscoveryLinkTreatmentType Enum=Default)
    ) -> (FSelectIslandEventData:FSelect Island Event Data Structure) {is=title K2Node_MakeStruct}
    EventContext:Object Reference
   ) -> (then, ReturnValue:bool)
   Event Destruct()
   > Unbind all Events from On Select Island Nav Left!(execute:Exec, self:WBP Custom Hover Content Disabled Quest Object Reference) -> (then:Exec) {is=title K2Node_ClearDelegate}
   > Unbind all Events from On Select Island Clicked!(execute:Exec, self:WBP Custom Hover Content Disabled Quest Object Reference) -> (then:Exec) {is=title K2Node_ClearDelegate}
  ]}
 IsInFrontend {nodes [
   IsInFrontend() {is=title K2Node_FunctionEntry}
   > n1 = Cast To FortPlayerControllerFrontEnd!(execute, Object=GetOwningPlayer()) -> (then, CastFailed, AsFort Player Controller Front End:Fort Player Controller Front End Object Reference) {is=title K2Node_DynamicCast}
   > Return Node!(execute, IsFrontend:bool=true)
   Return Node(execute=@n1.CastFailed, IsFrontend=false)
  ]}
 OnSelectIslandClicked {nodes [OnSelectIslandClicked!() -> (then, LinkID:Online Link Id Structure) {is=title K2Node_FunctionEntry}]}
 OnSelectIslandNavLeft {nodes [OnSelectIslandNavLeft() {is=title K2Node_FunctionEntry}]}
 SetDiscoveryLinkTreatmentType {nodes [
   n1 = SetDiscoveryLinkTreatmentType!() -> (then, NewTreatmentType:EDiscoveryLinkTreatmentType Enum=Default) {is=title K2Node_FunctionEntry}
   > n2 = Set DiscoveryLinkTreatmentType(DiscoveryLinkTreatmentType:EDiscoveryLinkTreatmentType Enum=@n1) -> (Output_Get:EDiscoveryLinkTreatmentType Enum=Default)
   > n3 = Switch on EDiscoveryLinkTreatmentType!(execute, Selection:EDiscoveryLinkTreatmentType Enum=@n2) -> (Default:Exec, Category:Exec) {is=title K2Node_SwitchEnum}
   Set Text(
    execute=@n3.Category
    Text="LOCTEXT(\"VIEW GENRE\")"
    self:Fort CTAButton Object Reference=Get Button_SelectIsland() -> (Button_SelectIsland:WBP UIKit Button Regular Object Reference)
   )
   Set Text(execute=@n3.Default, Text="LOCTEXT(\"SELECT ISLAND\")", self=Get Button_SelectIsland())
  ]}
 SetQuestOnlineLinkID {nodes [
   n1 = SetQuestOnlineLinkID!() -> (then, OnlineLinkID:Online Link Id Structure) {is=title K2Node_FunctionEntry}
   > n2 = Set OnlineLinkID!(execute, OnlineLinkID:Online Link Id Structure=@n1) -> (then, Output_Get_Mnemonic:String, Output_Get_Version:int)
   n3 = IsEmpty(InString=@n2)
   > n4 = Set IsShowingSelectButton(IsShowingSelectButton:bool=AND Boolean(A=NOT Boolean(A=@n3), B=IsInFrontend!() -> (IsFrontend:bool) {is=title K2Node_CallFunction})) -> (Output_Get:bool)
   > SetVisibility(self=Get Button_SelectIsland(), InVisibility=Select(Option 0=Collapsed, Option 1=Visible, Index=@n4))
  ]}
]
variables [
 DiscoveryLinkTreatmentType {type=EDiscoveryLinkTreatmentType, default=Default}
 IsShowingSelectButton {type=boolean}
 OnlineLinkID {type=OnlineLinkId, default {mnemonic="", version=-1}}
 OnSelectIslandClicked {}
 OnSelectIslandNavLeft {}
]
widgets {
 info {inheritedWidgetCount=0, parentClass=&UserWidget, rootWidgetClass=&Overlay, widgetCount=9}
 widgets [
  Overlay_25 {widgetClassPath=&Overlay, widget=Overlay_25}
  Image_QuestInfo_InvalidIsland {
   widgetClassPath=&Image
   bIsVariable=true
   parent=Overlay_25
   widget=Image_QuestInfo_InvalidIsland
   properties {
    renderOpacity=0.2
    brush {
     bIsDynamicallyLoaded=false
     drawAs=Image
     imageType=FullColor
     mirroring=NoMirror
     resourceName=None
     resourceObject=/QuestUI/QuestScreen/Textures/T_UI_TrackedQuestsIsland
     tiling=NoTile
     imageSize=<512 512>
     margin=<ltrb 0 0 0 0>
     outlineSettings {
      bUseBrushTransparency=false
      roundingType=HalfHeightRadius
      width=0
      color {colorUseRule=UseColor_Specified, specifiedColor=<rgba 0 0 0 0>}
      cornerRadii=<0 0 0 0>
     }
     tintColor {colorUseRule=UseColor_Specified, specifiedColor=<rgba 1 1 1 1>}
     uVRegion {bIsValid=false, max=<0 0>, min=<0 0>}
    }
    colorAndOpacity=<rgba 1 1 1 0.6>
   }
   slot {class=OverlaySlot, properties {horizontalAlignment=HAlign_Center, verticalAlignment=VAlign_Center}}
  }
  VerticalBox_0 {
   widgetClassPath=&VerticalBox
   parent=Overlay_25
   widget=VerticalBox_0
   slot {class=OverlaySlot, properties {horizontalAlignment=HAlign_Center, verticalAlignment=VAlign_Center, padding=<ltrb 0 260 0 0>}}
  }
  Overlay_0 {widgetClassPath=&Overlay, parent=VerticalBox_0, widget=Overlay_0, slot {class=VerticalBoxSlot}}
  Image_QuestInfo_DisableAlert {
   widgetClassPath=&Image
   bIsVariable=true
   parent=Overlay_0
   widget=Image_QuestInfo_DisableAlert
   properties {
    brush {
     bIsDynamicallyLoaded=false
     drawAs=Image
     imageType=FullColor
     mirroring=NoMirror
     resourceName=None
     resourceObject=/QuestUI/QuestScreen/Materials/MI_UI_QuestPin_DisabledAlert
     tiling=NoTile
     imageSize=<64 64>
     margin=<ltrb 0 0 0 0>
     outlineSettings {
      bUseBrushTransparency=false
      roundingType=HalfHeightRadius
      width=0
      color {colorUseRule=UseColor_Specified, specifiedColor=<rgba 0 0 0 0>}
      cornerRadii=<0 0 0 0>
     }
     tintColor {colorUseRule=UseColor_Specified, specifiedColor=<rgba 1 1 1 1>}
     uVRegion {bIsValid=false, max=<0 0>, min=<0 0>}
    }
    colorAndOpacity=<rgba 0.8714 0.8714 0.8714 1>
   }
   slot {class=OverlaySlot, properties {padding=<ltrb 0 0 10 0>}}
  }
  CommonTextBlock_58 {
   widgetClassPath=&CommonTextBlock
   parent=Overlay_0
   widget=CommonTextBlock_58
   properties {
    text=This Quest cannot be progressed on the selected island.
    wrapTextAt=500
    font {
     bForceMonospaced=false
     bMaterialIsStencil=false
     fontMaterial=null
     fontObject=&HeadingNow
     letterSpacing=0
     monospacedWidth=1
     size=18
     skewAmount=0
     typefaceFontName=75 Medium Italic
     outlineSettings {
      bApplyOutlineToDropShadows=false
      bMiteredCorners=false
      bSeparateFillAlpha=false
      outlineMaterial=null
      outlineSize=0
      outlineColor=<rgba 0 0 0 1>
     }
    }
    shadowOffset=<0 0>
   }
   slot {class=OverlaySlot, properties {verticalAlignment=VAlign_Fill, padding=<ltrb 65 0 0 0>}}
  }
  Overlay_1 {
   widgetClassPath=&Overlay
   parent=VerticalBox_0
   widget=Overlay_1
   slot {class=VerticalBoxSlot, properties {horizontalAlignment=HAlign_Left, verticalAlignment=VAlign_Bottom, padding=<ltrb 0 20 0 0>}}
  }
  WBP_CaptureForPostBufferUpdate {
   widgetClassPath=&WBP_CaptureForPostBufferUpdate
   bIsVariable=true
   parent=Overlay_1
   slot {class=OverlaySlot, properties {horizontalAlignment=HAlign_Fill, verticalAlignment=VAlign_Fill}}
  }
  Button_SelectIsland {
   widgetClassPath=&WBP_UIKit_Button_Regular
   bIsVariable=true
   parent=Overlay_1
   widget=Button_SelectIsland
   properties {
    buttonSize=Small
    buttonType=Quiet
    navigation=&self:WidgetTree.Button_SelectIsland.WidgetNavigation_0
    text=SELECT ISLAND
   }
   slot {class=OverlaySlot}
  }
 ]
}
